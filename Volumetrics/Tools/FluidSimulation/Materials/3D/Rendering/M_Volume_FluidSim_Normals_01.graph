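# Unlit, alpha-composite material that ray marches a fluid simulation volume inside a cube, with depth test disabled.
# A Custom node intersects the local-space box, clips to scene depth and optionally plane-aligns the entry, feeding start position and thickness-derived step count to the march.
# The march samples a flipbook pseudo-volume texture with temporal jitter, summing averaged RGB as density.
# Despite the Normals name, it computes no normals; transmittance stays at 1 and FinalStep is unused.
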Material /Volumetrics/Tools/FluidSimulation/Materials/3D/Rendering/M_Volume_FluidSim_Normals_01
graph {
 properties {bDisableDepthTest=true, blendMode=BLEND_AlphaComposite, pixelDepthOffsetMode=PDOM_Legacy, shadingModel=MSM_Unlit, usedShadingModels=Unlit}
 nodes [
  Constant() {r=1}
  MaterialExpressionScalarParameter_15 = ScalarParameter() {parameterName=Plane Alignment}
  MaterialExpressionScalarParameter_14 = ScalarParameter() {defaultValue=256, parameterName=Steps}
  n1 = Custom!(PlaneAlignment=@MaterialExpressionScalarParameter_15, MaxSteps=@MaterialExpressionScalarParameter_14) {
   code="\r\nfloat scale = length( TransformLocalVectorToWorld(Parameters, float3(1.00000000,0.00000000,0.00000000)).xyz);\r\nfloat localscenedepth = CalcSceneDepth(ScreenAlignedPosition(GetScreenPosition(Parameters)));\r\n\r\nfloat3 camerafwd = mul(float3(0.00000000,0.00000000,1.00000000),(float3x3)ResolvedView.ViewToTranslatedWorld);\r\nlocalscenedepth /= (GetPrimitiveData(Parameters).LocalObjectBoundsMax.x * 2 * scale);\r\nlocalscenedepth /= abs( dot( camerafwd, Parameters.CameraVector ) );\r\n\r\n//bring vectors into local space to support object transforms\r\nfloat3 localcampos = mul(float4( DFDemote(ResolvedView.WorldCameraOrigin),1.00000000), (WSDemote(GetWorldToLocal(Parameters)))).xyz;\r\nfloat3 localcamvec = -normalize( mul(Parameters.CameraVector, (float3x3)WSDemote(GetWorldToLocal(Parameters))) );\r\n\r\n//make camera position 0-1\r\nlocalcampos = (localcampos / (GetPrimitiveData(Parameters).LocalObjectBoundsMax.x * 2)) + 0.5;\r\n\r\nfloat3 invraydir = 1 / localcamvec;\r\n\r\nfloat3 firstintersections = (0 - localcampos) * invraydir;\r\nfloat3 secondintersections = (1 - localcampos) * invraydir;\r\nfloat3 closest = min(firstintersections, secondintersections);\r\nfloat3 furthest = max(firstintersections, secondintersections);\r\n\r\nfloat t0 = max(closest.x, max(closest.y, closest.z));\r\nfloat t1 = min(furthest.x, min(furthest.y, furthest.z));\r\n\r\nfloat planeoffset = 1-frac( ( t0 - length(localcampos-0.5) ) * MaxSteps );\r\n\r\nt0 += (planeoffset / MaxSteps) * PlaneAlignment;\r\nt1 = min(t1, localscenedepth);\r\nt0 = max(0, t0);\r\n\r\nfloat boxthickness = max(0, t1 - t0);\r\n\r\nfloat3 entrypos = localcampos + (max(0,t0) * localcamvec);\r\n\r\n\r\nreturn float4( entrypos, boxthickness );\r\n"
   desc="\n*** Original source before expression upgrade ***\n\r\nfloat scale = length( TransformLocalVectorToWorld(Parameters, float3(1.00000000,0.00000000,0.00000000)).xyz);\r\nfloat localscenedepth = CalcSceneDepth(ScreenAlignedPosition(GetScreenPosition(Parameters)));\r\n\r\nfloat3 camerafwd = mul(float3(0.00000000,0.00000000,1.00000000),(float3x3)ResolvedView.ViewToTranslatedWorld);\r\nlocalscenedepth /= (GetPrimitiveData(Parameters).LocalObjectBoundsMax.x * 2 * scale);\r\nlocalscenedepth /= abs( dot( camerafwd, Parameters.CameraVector ) );\r\n\r\n//bring vectors into local space to support object transforms\r\nfloat3 localcampos = mul(float4( ResolvedView.WorldCameraOrigin,1.00000000), (GetPrimitiveData(Parameters).WorldToLocal)).xyz;\r\nfloat3 localcamvec = -normalize( mul(Parameters.CameraVector, (float3x3)GetPrimitiveData(Parameters).WorldToLocal) );\r\n\r\n//make camera position 0-1\r\nlocalcampos = (localcampos / (GetPrimitiveData(Parameters).LocalObjectBoundsMax.x * 2)) + 0.5;\r\n\r\nfloat3 invraydir = 1 / localcamvec;\r\n\r\nfloat3 firstintersections = (0 - localcampos) * invraydir;\r\nfloat3 secondintersections = (1 - localcampos) * invraydir;\r\nfloat3 closest = min(firstintersections, secondintersections);\r\nfloat3 furthest = max(firstintersections, secondintersections);\r\n\r\nfloat t0 = max(closest.x, max(closest.y, closest.z));\r\nfloat t1 = min(furthest.x, min(furthest.y, furthest.z));\r\n\r\nfloat planeoffset = 1-frac( ( t0 - length(localcampos-0.5) ) * MaxSteps );\r\n\r\nt0 += (planeoffset / MaxSteps) * PlaneAlignment;\r\nt1 = min(t1, localscenedepth);\r\nt0 = max(0, t0);\r\n\r\nfloat boxthickness = max(0, t1 - t0);\r\n\r\nfloat3 entrypos = localcampos + (max(0,t0) * localcamvec);\r\n\r\n\r\nreturn float4( entrypos, boxthickness );\r\n"
   description=Ray March Cube Setup
   outputType=CMOT_Float4
   inputs [
    {
     inputName=PlaneAlignment
     input {expression=&self:MaterialExpressionScalarParameter_15}
    }
    {
     inputName=MaxSteps
     input {expression=&self:MaterialExpressionScalarParameter_14}
    }
   ]
  }
  n2 = Multiply(A=ComponentMask(None=@n1) {a=true}, B=ScalarParameter() {defaultValue=256, parameterName=Steps})
  Floor(None=@n2)
  MaterialExpressionTextureObjectParameter_3 = TextureObjectParameter() {parameterName=Color, samplerType=SAMPLERTYPE_LinearColor, texture=/Volumetrics/Content/VolumeTextures/Textures/RT_Default}
  MaterialExpressionComponentMask_2 = ComponentMask(None=VectorParameter() {parameterName=XYFrames, defaultValue=<rgba 16 16 0 0>}.RGB) {g=true, r=true}
  MaterialExpressionDivide_8 = Divide(B=ScalarParameter() {defaultValue=256, parameterName=Steps})
  MaterialExpressionClamp_0 = Clamp(None=Ceil(None=@n2)) {desc=MaxSteps, maxDefault=512}
  MaterialExpressionFrac_3 = Frac(None=@n2) {desc=FinalStepSize}
  MaterialExpressionMultiply_0 = Multiply(A=ComponentMask(None=@n1) {b=true, g=true, r=true}) {desc=Ray Entry}
  MaterialExpressionScalarParameter_0 = ScalarParameter() {defaultValue=1, parameterName=Density}
  MaterialExpressionScalarParameter_13 = ScalarParameter() {parameterName=Jitter}
  MaterialExpressionVectorParameter_18 = VectorParameter() {parameterName=VolumeResolution, defaultValue=<rgba 256 256 256 0>}
  n3 = Custom!(
   Tex=@MaterialExpressionTextureObjectParameter_3
   XYFrames=@MaterialExpressionComponentMask_2
   StepSize=@MaterialExpressionDivide_8
   MaxSteps=@MaterialExpressionClamp_0
   FinalStep=@MaterialExpressionFrac_3
   CurPos=@MaterialExpressionMultiply_0
   Density=@MaterialExpressionScalarParameter_0
   Jitter=@MaterialExpressionScalarParameter_13
   VolumeResolution=@MaterialExpressionVectorParameter_18.RGB
  ) {
   code="float numFrames = XYFrames.x * XYFrames.y;\r\nfloat accumdist = 0;\r\nfloat curdensity = 0;\r\nfloat transmittance = 1;\r\nfloat3 localcamvec = normalize( mul(Parameters.CameraVector, (float3x3)WSDemote(GetWorldToLocal(Parameters))) ) * StepSize;\r\n\r\n\r\n\r\nDensity *= StepSize;\r\nfloat3 lightenergy = 0;\r\n\r\nint3 randpos = int3(Parameters.SvPosition.xy, View.StateFrameIndexMod8);\r\nfloat rand =float(Rand3DPCG16(randpos).x) / 0xffff;\r\nCurPos +=  localcamvec * rand.x * Jitter;\r\n\r\n\r\nfor (int i = 0; i < MaxSteps; i++)\r\n{\t\r\n\tfloat3 color = PseudoVolumeTexture(Tex, TexSampler, clamp( CurPos, 0.5 / VolumeResolution, 1 - (0.5 / VolumeResolution) ), XYFrames, numFrames).rgb;\r\n\t\r\n\tcurdensity = saturate( dot(color, 0.333) * Density);\r\n\r\n\tlightenergy += curdensity;\t\r\n\r\n\tCurPos -= localcamvec;\r\n}\r\n\r\n\r\nreturn float4( lightenergy, transmittance);\r\n"
   desc="\n*** Original source before expression upgrade ***\nfloat numFrames = XYFrames.x * XYFrames.y;\r\nfloat accumdist = 0;\r\nfloat curdensity = 0;\r\nfloat transmittance = 1;\r\nfloat3 localcamvec = normalize( mul(Parameters.CameraVector, (float3x3)GetPrimitiveData(Parameters).WorldToLocal) ) * StepSize;\r\n\r\n\r\n\r\nDensity *= StepSize;\r\nfloat3 lightenergy = 0;\r\n\r\nint3 randpos = int3(Parameters.SvPosition.xy, View.StateFrameIndexMod8);\r\nfloat rand =float(Rand3DPCG16(randpos).x) / 0xffff;\r\nCurPos +=  localcamvec * rand.x * Jitter;\r\n\r\n\r\nfor (int i = 0; i < MaxSteps; i++)\r\n{\t\r\n\tfloat3 color = PseudoVolumeTexture(Tex, TexSampler, clamp( CurPos, 0.5 / VolumeResolution, 1 - (0.5 / VolumeResolution) ), XYFrames, numFrames).rgb;\r\n\t\r\n\tcurdensity = saturate( dot(color, 0.333) * Density);\r\n\r\n\tlightenergy += curdensity;\t\r\n\r\n\tCurPos -= localcamvec;\r\n}\r\n\r\n\r\nreturn float4( lightenergy, transmittance);\r\n"
   description=Density RayMarch
   outputType=CMOT_Float4
   inputs [
    {
     inputName=Tex
     input {expression=&self:MaterialExpressionTextureObjectParameter_3}
    }
    {
     inputName=XYFrames
     input {expression=&self:MaterialExpressionComponentMask_2}
    }
    {
     inputName=StepSize
     input {expression=&self:MaterialExpressionDivide_8}
    }
    {
     inputName=MaxSteps
     input {expression=&self:MaterialExpressionClamp_0}
    }
    {
     inputName=FinalStep
     input {expression=&self:MaterialExpressionFrac_3}
    }
    {
     inputName=CurPos
     input {expression=&self:MaterialExpressionMultiply_0}
    }
    {
     inputName=Density
     input {expression=&self:MaterialExpressionScalarParameter_0}
    }
    {
     inputName=Jitter
     input {expression=&self:MaterialExpressionScalarParameter_13}
    }
    {
     inputName=VolumeResolution
     input {
      expression=&self:MaterialExpressionVectorParameter_18
      mask=1
      maskB=1
      maskG=1
      maskR=1
     }
    }
   ]
  }
  OneMinus(None=ComponentMask(None=@n3) {a=true})
 ]
}
