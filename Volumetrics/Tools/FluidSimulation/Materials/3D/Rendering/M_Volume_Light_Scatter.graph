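# Unlit alpha-composite material that ray marches a fluid-simulation volume stored as a flipbook pseudo-volume texture.
# A custom node intersects the camera ray with the local bounds box, clipped by scene depth, giving entry and thickness.
# The temporally jittered march accumulates transmittance from RGB density and blackbody or grayscale emission from alpha.
# A separate Light Scatter node holds only placeholder code and feeds nothing visible.
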Material /Volumetrics/Tools/FluidSimulation/Materials/3D/Rendering/M_Volume_Light_Scatter
graph {
 properties {
  bAllowNegativeEmissiveColor=true
  bDisableDepthTest=true
  blendMode=BLEND_AlphaComposite
  pixelDepthOffsetMode=PDOM_Legacy
  shadingModel=MSM_Unlit
  usedShadingModels=Unlit
 }
 nodes [
  MaterialExpressionTextureObjectParameter_1 = TextureObjectParameter() {parameterName=Color, texture=/Volumetrics/Content/VolumeTextures/Textures/RT_Default}
  MaterialExpressionComponentMask_8 = ComponentMask(None=VectorParameter() {parameterName=XYFrames, defaultValue=<rgba 10 10 0 0>}.RGB) {g=true, r=true}
  MaterialExpressionDivide_1 = Divide(B=ScalarParameter() {defaultValue=512, parameterName=Steps})
  MaterialExpressionScalarParameter_19 = ScalarParameter() {defaultValue=1, parameterName=Plane Alignment}
  MaterialExpressionScalarParameter_18 = ScalarParameter() {defaultValue=512, parameterName=Steps}
  n1 = Custom!(PlaneAlignment=@MaterialExpressionScalarParameter_19, MaxSteps=@MaterialExpressionScalarParameter_18) {
   code="float localscenedepth = CalcSceneDepth(ScreenAlignedPosition(GetScreenPosition(Parameters)));\r\n\r\nfloat3 camerafwd = mul(float3(0.00000000,0.00000000,1.00000000),ResolvedView.ViewToTranslatedWorld);\r\n\r\nfloat3 depthpos = ((Parameters.CameraVector * localscenedepth) / abs( dot( camerafwd, Parameters.CameraVector ) ) );\r\n\r\ndepthpos = mul(depthpos,  (WSDemote(GetWorldToLocal(Parameters)))).xyz;\r\n\r\ndepthpos /= 256;\r\n\r\nlocalscenedepth = length(depthpos);\r\n\r\n//0-1\r\n//localscenedepth /= (GetPrimitiveData(Parameters).LocalObjectBoundsMax.x * 2 * scale);\r\n//localscenedepth /= abs( dot( camerafwd, Parameters.CameraVector ) );\r\n\r\n//bring vectors into local space to support object transforms\r\nfloat3 localcampos = mul(float4( DFDemote(ResolvedView.WorldCameraOrigin),1.00000000), (WSDemote(GetWorldToLocal(Parameters)))).xyz;\r\nfloat3 localcamvec = -normalize( mul(Parameters.CameraVector, WSDemote(GetWorldToLocal(Parameters))) );\r\n\r\n//make camera position 0-1\r\nlocalcampos = (localcampos / (GetPrimitiveData(Parameters).LocalObjectBoundsMax.x * 2)) + 0.5;\r\n\r\nfloat3 invraydir = 1 / localcamvec;\r\n\r\nfloat3 firstintersections = (0 - localcampos) * invraydir;\r\nfloat3 secondintersections = (1 - localcampos) * invraydir;\r\nfloat3 closest = min(firstintersections, secondintersections);\r\nfloat3 furthest = max(firstintersections, secondintersections);\r\n\r\nfloat t0 = max(closest.x, max(closest.y, closest.z));\r\nfloat t1 = min(furthest.x, min(furthest.y, furthest.z));\r\n\r\nfloat planeoffset = 1-frac( ( t0 - length(localcampos-0.5) ) * MaxSteps );\r\n\r\nt0 += (planeoffset / MaxSteps) * PlaneAlignment;\r\nt1 = min(t1, localscenedepth);\r\nt0 = max(0, t0);\r\n\r\nfloat boxthickness = max(0, t1 - t0);\r\n\r\nfloat3 entrypos = localcampos + (max(0,t0) * localcamvec);\r\n\r\n\r\nreturn float4( entrypos, boxthickness );"
   desc="\n*** Original source before expression upgrade ***\nfloat localscenedepth = CalcSceneDepth(ScreenAlignedPosition(GetScreenPosition(Parameters)));\r\n\r\nfloat3 camerafwd = mul(float3(0.00000000,0.00000000,1.00000000),ResolvedView.ViewToTranslatedWorld);\r\n\r\nfloat3 depthpos = ((Parameters.CameraVector * localscenedepth) / abs( dot( camerafwd, Parameters.CameraVector ) ) );\r\n\r\ndepthpos = mul(depthpos,  (GetPrimitiveData(Parameters).WorldToLocal)).xyz;\r\n\r\ndepthpos /= 256;\r\n\r\nlocalscenedepth = length(depthpos);\r\n\r\n//0-1\r\n//localscenedepth /= (GetPrimitiveData(Parameters).LocalObjectBoundsMax.x * 2 * scale);\r\n//localscenedepth /= abs( dot( camerafwd, Parameters.CameraVector ) );\r\n\r\n//bring vectors into local space to support object transforms\r\nfloat3 localcampos = mul(float4( ResolvedView.WorldCameraOrigin,1.00000000), (GetPrimitiveData(Parameters).WorldToLocal)).xyz;\r\nfloat3 localcamvec = -normalize( mul(Parameters.CameraVector, GetPrimitiveData(Parameters).WorldToLocal) );\r\n\r\n//make camera position 0-1\r\nlocalcampos = (localcampos / (GetPrimitiveData(Parameters).LocalObjectBoundsMax.x * 2)) + 0.5;\r\n\r\nfloat3 invraydir = 1 / localcamvec;\r\n\r\nfloat3 firstintersections = (0 - localcampos) * invraydir;\r\nfloat3 secondintersections = (1 - localcampos) * invraydir;\r\nfloat3 closest = min(firstintersections, secondintersections);\r\nfloat3 furthest = max(firstintersections, secondintersections);\r\n\r\nfloat t0 = max(closest.x, max(closest.y, closest.z));\r\nfloat t1 = min(furthest.x, min(furthest.y, furthest.z));\r\n\r\nfloat planeoffset = 1-frac( ( t0 - length(localcampos-0.5) ) * MaxSteps );\r\n\r\nt0 += (planeoffset / MaxSteps) * PlaneAlignment;\r\nt1 = min(t1, localscenedepth);\r\nt0 = max(0, t0);\r\n\r\nfloat boxthickness = max(0, t1 - t0);\r\n\r\nfloat3 entrypos = localcampos + (max(0,t0) * localcamvec);\r\n\r\n\r\nreturn float4( entrypos, boxthickness );"
   description=Ray March Cube Setup
   outputType=CMOT_Float4
   inputs [
    {
     inputName=PlaneAlignment
     input {expression=&self:MaterialExpressionScalarParameter_19}
    }
    {
     inputName=MaxSteps
     input {expression=&self:MaterialExpressionScalarParameter_18}
    }
   ]
  }
  n2 = Multiply(A=ComponentMask(None=@n1) {a=true}, B=ScalarParameter() {defaultValue=512, parameterName=Steps})
  MaterialExpressionClamp_1 = Clamp(None=Floor(None=@n2)) {desc=MaxSteps, maxDefault=512}
  MaterialExpressionFrac_0 = Frac(None=@n2) {desc=FinalStepSize}
  MaterialExpressionMultiply_1 = Multiply(A=ComponentMask(None=@n1) {b=true, g=true, r=true}) {desc=Ray Entry}
  MaterialExpressionScalarParameter_9 = ScalarParameter() {defaultValue=16, parameterName=Density}
  MaterialExpressionMultiply_4 = Multiply(
   A=MaterialFunctionCall(Alpha Threshold=ScalarParameter() {defaultValue=0.5, parameterName=Jitter Alpha Threshold}) {materialFunction=&DitherTemporalAA}.Result
   B=ScalarParameter() {defaultValue=1, parameterName=Jitter}
  )
  MaterialExpressionScalarParameter_27 = ScalarParameter() {defaultValue=15000, parameterName=Temperature}
  MaterialExpressionVectorParameter_1 = VectorParameter() {parameterName=VolumeResolution, defaultValue=<rgba 256 256 256 0>}
  MaterialExpressionScalarParameter_32 = ScalarParameter() {parameterName=Use Blackbody or Grayscale}
  n3 = Custom!(
   Tex=@MaterialExpressionTextureObjectParameter_1
   XYFrames=@MaterialExpressionComponentMask_8
   StepSize=@MaterialExpressionDivide_1
   MaxSteps=@MaterialExpressionClamp_1
   FinalStep=@MaterialExpressionFrac_0
   CurPos=@MaterialExpressionMultiply_1
   Density=@MaterialExpressionScalarParameter_9
   Jitter=@MaterialExpressionMultiply_4
   Temp=@MaterialExpressionScalarParameter_27
   VolumeResolution=@MaterialExpressionVectorParameter_1.RGB
   UseBlackBodyColors=@MaterialExpressionScalarParameter_32
  ) {
   code="##define EnableCurl 0\r\nfloat numFrames = XYFrames.x * XYFrames.y;\r\nfloat accumdist = 0;\r\nfloat3 curdensity = 0;\r\nfloat transmittance = 1;\r\nfloat3 localcamvec = normalize( mul(Parameters.CameraVector, WSDemote(GetWorldToLocal(Parameters))) ) * StepSize;\r\n\r\nDensity *= StepSize;\r\nfloat3 lightenergy = 0;\r\n\r\nCurPos +=  localcamvec * Jitter;\r\n\r\nfor (int i = 0; i < MaxSteps; i++)\r\n{\t\r\n\tfloat3 samplepos =  saturate(CurPos);\r\n\tsamplepos.z = min(samplepos.z, 1 - (1/VolumeResolution.z));\r\n\r\n\tfloat4 cursample = PseudoVolumeTexture(Tex, TexSampler, samplepos, XYFrames, numFrames);\r\n\r\n\t//Sample Light Absorption and Scattering\r\n\tif( cursample.r + cursample.g + cursample.b > 0.00001)\r\n\t{\r\n\r\n\t\tcurdensity = dot(cursample.rgb, 0.333);\r\n\t\tcurdensity = 1 - exp(-curdensity * Density);\r\n\t\t\r\n\t\tfloat3 curenergy = UseBlackBodyColors == 1.0 ? MaterialExpressionBlackBody(Temp*cursample.a) : cursample.a;\r\n\t\t\r\n\t\tlightenergy += curenergy * curdensity * transmittance;\r\n\t\t\r\n\t\ttransmittance *= 1-curdensity;\r\n\t}\r\n\t\t\r\n\tCurPos -= localcamvec;\r\n}\r\n\r\nreturn float4( lightenergy, transmittance);\r\n"
   desc="\n*** Original source before expression upgrade ***\n##define EnableCurl 0\r\nfloat numFrames = XYFrames.x * XYFrames.y;\r\nfloat accumdist = 0;\r\nfloat3 curdensity = 0;\r\nfloat transmittance = 1;\r\nfloat3 localcamvec = normalize( mul(Parameters.CameraVector, GetPrimitiveData(Parameters).WorldToLocal) ) * StepSize;\r\n\r\nDensity *= StepSize;\r\nfloat3 lightenergy = 0;\r\n\r\nCurPos +=  localcamvec * Jitter;\r\n\r\nfor (int i = 0; i < MaxSteps; i++)\r\n{\t\r\n\tfloat3 samplepos =  saturate(CurPos);\r\n\tsamplepos.z = min(samplepos.z, 1 - (1/VolumeResolution.z));\r\n\r\n\tfloat4 cursample = PseudoVolumeTexture(Tex, TexSampler, samplepos, XYFrames, numFrames);\r\n\r\n\t//Sample Light Absorption and Scattering\r\n\tif( cursample.r + cursample.g + cursample.b > 0.00001)\r\n\t{\r\n\r\n\t\tcurdensity = dot(cursample.rgb, 0.333);\r\n\t\tcurdensity = 1 - exp(-curdensity * Density);\r\n\t\t\r\n\t\tfloat3 curenergy = UseBlackBodyColors == 1.0 ? MaterialExpressionBlackBody(Temp*cursample.a) : cursample.a;\r\n\t\t\r\n\t\tlightenergy += curenergy * curdensity * transmittance;\r\n\t\t\r\n\t\ttransmittance *= 1-curdensity;\r\n\t}\r\n\t\t\r\n\tCurPos -= localcamvec;\r\n}\r\n\r\nreturn float4( lightenergy, transmittance);\r\n"
   description=Density RayMarch
   outputType=CMOT_Float4
   inputs [
    {
     inputName=Tex
     input {expression=&self:MaterialExpressionTextureObjectParameter_1}
    }
    {
     inputName=XYFrames
     input {expression=&self:MaterialExpressionComponentMask_8}
    }
    {
     inputName=StepSize
     input {expression=&self:MaterialExpressionDivide_1}
    }
    {
     inputName=MaxSteps
     input {expression=&self:MaterialExpressionClamp_1}
    }
    {
     inputName=FinalStep
     input {expression=&self:MaterialExpressionFrac_0}
    }
    {
     inputName=CurPos
     input {expression=&self:MaterialExpressionMultiply_1}
    }
    {
     inputName=Density
     input {expression=&self:MaterialExpressionScalarParameter_9}
    }
    {
     inputName=Jitter
     input {expression=&self:MaterialExpressionMultiply_4}
    }
    {
     inputName=Temp
     input {expression=&self:MaterialExpressionScalarParameter_27}
    }
    {
     inputName=VolumeResolution
     input {
      expression=&self:MaterialExpressionVectorParameter_1
      mask=1
      maskB=1
      maskG=1
      maskR=1
     }
    }
    {
     inputName=UseBlackBodyColors
     input {expression=&self:MaterialExpressionScalarParameter_32}
    }
   ]
  }
  ComponentMask(None=@n3) {b=true, g=true, r=true}
  MaterialExpressionTextureObjectParameter_0 = TextureObjectParameter() {parameterName=Color, samplerType=SAMPLERTYPE_LinearColor, texture=&RT_Default}
  MaterialExpressionComponentMask_0 = ComponentMask(None=VectorParameter() {parameterName=XYFrames, defaultValue=<rgba 10 10 0 0>}.RGB) {g=true, r=true}
  MaterialExpressionMaterialFunctionCall_0 = MaterialFunctionCall!(UVs, XY Frames=@MaterialExpressionComponentMask_0) {materialFunction=/Volumetrics/Content/VolumeTextures/MaterialFunctions/EncodeVolumeCoordinates}
  MaterialExpressionScalarParameter_0 = ScalarParameter() {defaultValue=512, parameterName=Steps}
  MaterialExpressionScalarParameter_1 = ScalarParameter() {defaultValue=0.25, parameterName=Trace Distance}
  MaterialExpressionScalarParameter_2 = ScalarParameter() {defaultValue=16, parameterName=Density}
  MaterialExpressionScalarParameter_5 = ScalarParameter() {defaultValue=15000, parameterName=Temperature}
  Custom!(
   Tex=@MaterialExpressionTextureObjectParameter_0
   Pos=@MaterialExpressionMaterialFunctionCall_0.Result
   XYFrames=@MaterialExpressionComponentMask_0
   MaxSteps=@MaterialExpressionScalarParameter_0
   TraceDistance=@MaterialExpressionScalarParameter_1
   Density=@MaterialExpressionScalarParameter_2
   Temperature=@MaterialExpressionScalarParameter_5
  ) {
   code=Light Scatter
   description=Light Scatter
   inputs [
    {
     inputName=Tex
     input {expression=&self:MaterialExpressionTextureObjectParameter_0}
    }
    {
     inputName=Pos
     input {expression=&self:MaterialExpressionMaterialFunctionCall_0}
    }
    {
     inputName=XYFrames
     input {expression=&self:MaterialExpressionComponentMask_0}
    }
    {
     inputName=MaxSteps
     input {expression=&self:MaterialExpressionScalarParameter_0}
    }
    {
     inputName=TraceDistance
     input {expression=&self:MaterialExpressionScalarParameter_1}
    }
    {
     inputName=Density
     input {expression=&self:MaterialExpressionScalarParameter_2}
    }
    {
     inputName=Temperature
     input {expression=&self:MaterialExpressionScalarParameter_5}
    }
   ]
  }
  OneMinus(None=ComponentMask(None=@n3) {a=true})
  VectorNoise()
 ]
}
