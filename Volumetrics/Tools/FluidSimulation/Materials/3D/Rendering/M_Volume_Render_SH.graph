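# Unlit alpha-composite material that ray-marches a fluid-simulation volume for what appear to be spherical-harmonic passes, per its SH naming.
# A box intersection sets entry and step count for a custom HLSL loop over a flipbook pseudo-volume texture, with temporal jitter.
# The loop accumulates shadow-marched Schlick-phase sunlight, blackbody emission and three-tap sky occlusion.
# Its lighting is desaturated, then SHIndex selects how that luminance is distributed across channels.
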Material /Volumetrics/Tools/FluidSimulation/Materials/3D/Rendering/M_Volume_Render_SH
graph {
 properties {bDisableDepthTest=true, blendMode=BLEND_AlphaComposite, pixelDepthOffsetMode=PDOM_Legacy, shadingModel=MSM_Unlit, usedShadingModels=Unlit}
 nodes [
  MaterialExpressionTextureObjectParameter_0 = TextureObjectParameter() {parameterName=Color, texture=/Volumetrics/Content/VolumeTextures/Textures/RT_Default}
  MaterialExpressionComponentMask_8 = ComponentMask(None=VectorParameter() {parameterName=XYFrames, defaultValue=<rgba 10 10 0 0>}.RGB) {g=true, r=true}
  MaterialExpressionDivide_0 = Divide(B=ScalarParameter() {defaultValue=128, parameterName=Steps})
  n1 = MaterialFunctionCall!(
   Steps=ScalarParameter() {defaultValue=128, parameterName=Steps}
   Plane alignment.=ScalarParameter() {defaultValue=1, parameterName=Plane Alignment}
  ) {materialFunction=/Volumetrics/Content/VolumeTextures/MaterialFunctions/VolumeBoxIntersect}
  n2 = Multiply(A=@n1.Box Distance, B=ScalarParameter() {defaultValue=128, parameterName=Steps})
  MaterialExpressionClamp_1 = Clamp(None=Floor(None=@n2)) {desc=MaxSteps, maxDefault=512}
  MaterialExpressionFrac_0 = Frac(None=@n2) {desc=FinalStepSize}
  MaterialExpressionMultiply_2 = Multiply(A=@n1.Intersection Position) {desc=Ray Entry}
  MaterialExpressionScalarParameter_11 = ScalarParameter() {defaultValue=16, parameterName=Density}
  MaterialExpressionScalarParameter_8 = ScalarParameter() {defaultValue=16, parameterName=ShadowSteps}
  MaterialExpressionScalarParameter_10 = ScalarParameter() {defaultValue=2, parameterName=ShadowDensity}
  MaterialExpressionNormalize_0 = Normalize(VectorInput=Transform(None=VectorParameter() {parameterName=LightVector, defaultValue=<rgba 1 0 1 1>}.RGB) {transformSourceType=TRANSFORMSOURCE_World, transformType=TRANSFORM_Local})
  MaterialExpressionScalarParameter_16 = ScalarParameter() {defaultValue=0.01, parameterName=SB}
  n3 = VectorParameter() {parameterName=Sun Color, defaultValue=<rgba 1 0.875 0.55 4>}
  n4 = DotProduct(A=Normalize(VectorInput=VectorParameter() {parameterName=LightVector, defaultValue=<rgba 1 0 1 1>}.RGB), B=CameraVectorWS())
  n5 = ScalarParameter() {parameterName=Aniso}
  MaterialExpressionMultiply_5 = Multiply(
   A=@n3.RGB
   B=Multiply(
    A=@n3.A
    B=MaterialFunctionCall!(Cosine=@n4, Anisotropy=@n5) {materialFunction=/Engine/Functions/Engine_MaterialFunctions01/Volumetrics/SchlickPhase}.Result
   )
  )
  n6 = VectorParameter() {parameterName=Sky Color, defaultValue=<rgba 0.2 0.2 0.25 0.2>}
  MaterialExpressionMultiply_6 = Multiply(A=@n6.RGB, B=@n6.RGB)
  MaterialExpressionScalarParameter_17 = ScalarParameter() {defaultValue=0.1, parameterName=Ambient Density}
  MaterialExpressionMultiply_8 = Multiply(
   A=MaterialFunctionCall(Alpha Threshold=ScalarParameter() {defaultValue=0.5, parameterName=Jitter Alpha Threshold}) {materialFunction=&DitherTemporalAA}.Result
   B=ScalarParameter() {defaultValue=1, parameterName=Jitter}
  )
  MaterialExpressionScalarParameter_27 = ScalarParameter() {parameterName=Temperature}
  MaterialExpressionScalarParameter_31 = ScalarParameter() {defaultValue=0.75, parameterName=Light Offset}
  MaterialExpressionVectorParameter_12 = VectorParameter() {parameterName=VolumeResolution, defaultValue=<rgba 256 256 256 0>}
  n7 = Custom!(
   Tex=@MaterialExpressionTextureObjectParameter_0
   XYFrames=@MaterialExpressionComponentMask_8
   StepSize=@MaterialExpressionDivide_0
   MaxSteps=@MaterialExpressionClamp_1
   FinalStep=@MaterialExpressionFrac_0
   CurPos=@MaterialExpressionMultiply_2
   Density=@MaterialExpressionScalarParameter_11
   ShadowSteps=@MaterialExpressionScalarParameter_8
   ShadowDensity=@MaterialExpressionScalarParameter_10
   LightVector=@MaterialExpressionNormalize_0
   ShadowThreshold=@MaterialExpressionScalarParameter_16
   LightColor=@MaterialExpressionMultiply_5
   SkyColor=@MaterialExpressionMultiply_6
   AmbientDensity=@MaterialExpressionScalarParameter_17
   Jitter=@MaterialExpressionMultiply_8
   Temp=@MaterialExpressionScalarParameter_27
   LightOffset=@MaterialExpressionScalarParameter_31
   VolumeResolution=@MaterialExpressionVectorParameter_12.RGB
  ) {
   code="float numFrames = XYFrames.x * XYFrames.y;\r\nfloat accumdist = 0;\r\nfloat3 curdensity = 0;\r\nfloat transmittance = 1;\r\nfloat3 localcamvec = normalize( mul(Parameters.CameraVector, (float3x3)WSDemote(GetWorldToLocal(Parameters))) ) * StepSize;\r\n\r\nfloat3 invlightdir = 1 / LightVector;\r\n\r\nfloat shadowstepsize = 1 / ShadowSteps;\r\nLightVector *= shadowstepsize*0.25;\r\nShadowDensity *= shadowstepsize;\r\n\r\nDensity *= StepSize;\r\nfloat3 lightenergy = 0;\r\nfloat shadowthresh = -log(ShadowThreshold) / ShadowDensity;\r\n\r\nCurPos +=  localcamvec * Jitter;\r\n\r\nfor (int i = 0; i < MaxSteps; i++)\r\n{\r\n\r\n\tfloat3 samplepos =  saturate(CurPos);\r\n\tsamplepos.z = min(samplepos.z, 1 - (1/VolumeResolution.z));\r\n\r\n\tfloat4 cursample = PseudoVolumeTexture(Tex, TexSampler, samplepos, XYFrames, numFrames);\r\n\r\n\r\n\t//Sample Light Absorption and Scattering\r\n\tif( cursample.r + cursample.g + cursample.b > 0.001)\r\n\t{\r\n\t\tfloat3 lpos = CurPos - LightVector * LightOffset;\r\n\t\tfloat shadowdist = 0;\r\n\r\n\t\tfor (int s = 0; s < ShadowSteps; s++)\r\n\t\t{\r\n\t\t\tlpos += LightVector;\r\n\t\t\r\n\t\t\tfloat3 lsample = PseudoVolumeTexture(Tex, TexSampler, saturate(lpos), XYFrames, numFrames).rgb;\r\n\t\t\t\r\n\t\t\tfloat3 shadowboxtest = floor( 0.5 + ( abs( 0.5 - lpos ) ) );\r\n\t\t\tfloat exitshadowbox = shadowboxtest .x + shadowboxtest .y + shadowboxtest .z;\r\n\r\n       \t\t\tif(shadowdist > shadowthresh || exitshadowbox >= 1) break;\r\n\r\n\t\t\tshadowdist += dot(lsample, 0.3333);\r\n\t\t}\r\n\t\r\n\t\tcurdensity = 1 - exp(-cursample.rgb * Density);\r\n\t\tlightenergy += exp(-shadowdist * ShadowDensity) * curdensity * transmittance * LightColor;\r\n\t\tlightenergy += (MaterialExpressionBlackBody(Temp*cursample.a)) * curdensity * transmittance;\r\n\t\t//lightenergy += pressure * transmittance * PressureDisplay;\r\n\r\n\t\t\r\n\t\ttransmittance *= 1-dot(curdensity, 0.333);\r\n\t\t\t\t##if 1\r\n\r\n\t\t//Sky Lighting\r\n\t\tshadowdist = 0;\r\n\t\t\r\n\t\tlpos = CurPos + float3(0,0,0.025);\r\n\t\tfloat3 lsample = PseudoVolumeTexture(Tex, TexSampler, saturate(lpos), XYFrames, numFrames).rgb;\r\n\t\tshadowdist += dot(lsample, 0.3333);\r\n\t\tlpos = CurPos + float3(0,0,0.05);\r\n\t\tlsample = PseudoVolumeTexture(Tex, TexSampler, saturate(lpos), XYFrames, numFrames).rgb;\r\n\t\tshadowdist += dot(lsample, 0.3333);\r\n\t\tlpos = CurPos + float3(0,0,0.15);\r\n\t\tlsample = PseudoVolumeTexture(Tex, TexSampler, saturate(lpos), XYFrames, numFrames).rgb;\r\n\t\tshadowdist += dot(lsample, 0.3333);\r\n\r\n\t\t//shadowterm    = exp(-shadowdist  * AmbientDensity);\r\n\t\t//absorbedlight = exp(-shadowdist  * AmbientDensity) * curdensity;\r\n\t\tlightenergy    += exp(-shadowdist  * AmbientDensity) * curdensity * SkyColor * transmittance;\r\n\t\t##endif\r\n\r\n\t}\r\n\t\t\r\n\r\n\tCurPos -= localcamvec;\r\n}\r\n\r\n\r\n\r\nreturn float4( lightenergy, transmittance);\r\n"
   desc="\n*** Original source before expression upgrade ***\nfloat numFrames = XYFrames.x * XYFrames.y;\r\nfloat accumdist = 0;\r\nfloat3 curdensity = 0;\r\nfloat transmittance = 1;\r\nfloat3 localcamvec = normalize( mul(Parameters.CameraVector, (float3x3)GetPrimitiveData(Parameters).WorldToLocal) ) * StepSize;\r\n\r\nfloat3 invlightdir = 1 / LightVector;\r\n\r\nfloat shadowstepsize = 1 / ShadowSteps;\r\nLightVector *= shadowstepsize*0.25;\r\nShadowDensity *= shadowstepsize;\r\n\r\nDensity *= StepSize;\r\nfloat3 lightenergy = 0;\r\nfloat shadowthresh = -log(ShadowThreshold) / ShadowDensity;\r\n\r\nCurPos +=  localcamvec * Jitter;\r\n\r\nfor (int i = 0; i < MaxSteps; i++)\r\n{\r\n\r\n\tfloat3 samplepos =  saturate(CurPos);\r\n\tsamplepos.z = min(samplepos.z, 1 - (1/VolumeResolution.z));\r\n\r\n\tfloat4 cursample = PseudoVolumeTexture(Tex, TexSampler, samplepos, XYFrames, numFrames);\r\n\r\n\r\n\t//Sample Light Absorption and Scattering\r\n\tif( cursample.r + cursample.g + cursample.b > 0.001)\r\n\t{\r\n\t\tfloat3 lpos = CurPos - LightVector * LightOffset;\r\n\t\tfloat shadowdist = 0;\r\n\r\n\t\tfor (int s = 0; s < ShadowSteps; s++)\r\n\t\t{\r\n\t\t\tlpos += LightVector;\r\n\t\t\r\n\t\t\tfloat3 lsample = PseudoVolumeTexture(Tex, TexSampler, saturate(lpos), XYFrames, numFrames).rgb;\r\n\t\t\t\r\n\t\t\tfloat3 shadowboxtest = floor( 0.5 + ( abs( 0.5 - lpos ) ) );\r\n\t\t\tfloat exitshadowbox = shadowboxtest .x + shadowboxtest .y + shadowboxtest .z;\r\n\r\n       \t\t\tif(shadowdist > shadowthresh || exitshadowbox >= 1) break;\r\n\r\n\t\t\tshadowdist += dot(lsample, 0.3333);\r\n\t\t}\r\n\t\r\n\t\tcurdensity = 1 - exp(-cursample.rgb * Density);\r\n\t\tlightenergy += exp(-shadowdist * ShadowDensity) * curdensity * transmittance * LightColor;\r\n\t\tlightenergy += (MaterialExpressionBlackBody(Temp*cursample.a)) * curdensity * transmittance;\r\n\t\t//lightenergy += pressure * transmittance * PressureDisplay;\r\n\r\n\t\t\r\n\t\ttransmittance *= 1-dot(curdensity, 0.333);\r\n\t\t\t\t##if 1\r\n\r\n\t\t//Sky Lighting\r\n\t\tshadowdist = 0;\r\n\t\t\r\n\t\tlpos = CurPos + float3(0,0,0.025);\r\n\t\tfloat3 lsample = PseudoVolumeTexture(Tex, TexSampler, saturate(lpos), XYFrames, numFrames).rgb;\r\n\t\tshadowdist += dot(lsample, 0.3333);\r\n\t\tlpos = CurPos + float3(0,0,0.05);\r\n\t\tlsample = PseudoVolumeTexture(Tex, TexSampler, saturate(lpos), XYFrames, numFrames).rgb;\r\n\t\tshadowdist += dot(lsample, 0.3333);\r\n\t\tlpos = CurPos + float3(0,0,0.15);\r\n\t\tlsample = PseudoVolumeTexture(Tex, TexSampler, saturate(lpos), XYFrames, numFrames).rgb;\r\n\t\tshadowdist += dot(lsample, 0.3333);\r\n\r\n\t\t//shadowterm    = exp(-shadowdist  * AmbientDensity);\r\n\t\t//absorbedlight = exp(-shadowdist  * AmbientDensity) * curdensity;\r\n\t\tlightenergy    += exp(-shadowdist  * AmbientDensity) * curdensity * SkyColor * transmittance;\r\n\t\t##endif\r\n\r\n\t}\r\n\t\t\r\n\r\n\tCurPos -= localcamvec;\r\n}\r\n\r\n\r\n\r\nreturn float4( lightenergy, transmittance);\r\n"
   description=Density RayMarch
   outputType=CMOT_Float4
   inputs [
    {
     inputName=Tex
     input {expression=&self:MaterialExpressionTextureObjectParameter_0}
    }
    {
     inputName=XYFrames
     input {expression=&self:MaterialExpressionComponentMask_8}
    }
    {
     inputName=StepSize
     input {expression=&self:MaterialExpressionDivide_0}
    }
    {
     inputName=MaxSteps
     input {expression=&self:MaterialExpressionClamp_1}
    }
    {
     inputName=FinalStep
     input {expression=&self:MaterialExpressionFrac_0}
    }
    {
     inputName=CurPos
     input {expression=&self:MaterialExpressionMultiply_2}
    }
    {
     inputName=Density
     input {expression=&self:MaterialExpressionScalarParameter_11}
    }
    {
     inputName=ShadowSteps
     input {expression=&self:MaterialExpressionScalarParameter_8}
    }
    {
     inputName=ShadowDensity
     input {expression=&self:MaterialExpressionScalarParameter_10}
    }
    {
     inputName=LightVector
     input {expression=&self:MaterialExpressionNormalize_0}
    }
    {
     inputName=ShadowThreshold
     input {expression=&self:MaterialExpressionScalarParameter_16}
    }
    {
     inputName=LightColor
     input {expression=&self:MaterialExpressionMultiply_5}
    }
    {
     inputName=SkyColor
     input {expression=&self:MaterialExpressionMultiply_6}
    }
    {
     inputName=AmbientDensity
     input {expression=&self:MaterialExpressionScalarParameter_17}
    }
    {
     inputName=Jitter
     input {expression=&self:MaterialExpressionMultiply_8}
    }
    {
     inputName=Temp
     input {expression=&self:MaterialExpressionScalarParameter_27}
    }
    {
     inputName=LightOffset
     input {expression=&self:MaterialExpressionScalarParameter_31}
    }
    {
     inputName=VolumeResolution
     input {
      expression=&self:MaterialExpressionVectorParameter_12
      mask=1
      maskB=1
      maskG=1
      maskR=1
     }
    }
   ]
  }
  n8 = Desaturation(None=ComponentMask(None=@n7) {b=true, g=true, r=true}, Fraction=Constant() {r=1})
  n9 = Constant() {r=1}
  MaterialFunctionCall(1=@n8, 2=@n9, 3=@n9, 4=@n9, A=Divide(A=ScalarParameter() {parameterName=SHIndex}) {constB=3}) {materialFunction=&Lerp_Multiple_Float3}
  n10 = ScalarParameter() {parameterName=SHIndex}
  Multiply(
   A=MaterialFunctionCall(
    1=Constant()
    2=Constant3Vector() {constant=<rgba 0 0 1 0>}
    3=Constant3Vector() {constant=<rgba 0 1 0 0>}
    4=Constant3Vector() {constant=<rgba 1 0 0 0>}
    A=Divide(A=@n10) {constB=3}
   ) {materialFunction=&Lerp_Multiple_Float3}.Lerp 4 Inputs
   B=@n8
  )
 ]
}
