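# Unlit, alpha-composite volume material that renders temperature from a fluid simulation by ray-marching a flipbook-style pseudo-volume texture.
# VolumeBoxIntersect supplies ray entry and step count to a custom HLSL loop, with DitherTemporalAA jittering the start.
# Averaged RGB drives exponential absorption, while alpha becomes blackbody color or grayscale emission, accumulated front to back.
# The loop returns light energy and transmittance, the latter inverted for opacity; its FinalStep input goes unused.
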Material /Volumetrics/Tools/FluidSimulation/Materials/3D/Rendering/M_Volume_Render_Temperature
graph {
 properties {
  bAllowNegativeEmissiveColor=true
  bDisableDepthTest=true
  blendMode=BLEND_AlphaComposite
  pixelDepthOffsetMode=PDOM_Legacy
  shadingModel=MSM_Unlit
  usedShadingModels=Unlit
 }
 nodes [
  MaterialExpressionTextureObjectParameter_1 = TextureObjectParameter() {parameterName=Color, texture=/Volumetrics/Content/VolumeTextures/Textures/RT_Default}
  MaterialExpressionComponentMask_8 = ComponentMask(None=VectorParameter() {parameterName=XYFrames, defaultValue=<rgba 10 10 0 0>}.RGB) {g=true, r=true}
  MaterialExpressionDivide_1 = Divide(B=ScalarParameter() {defaultValue=512, parameterName=Steps})
  n1 = MaterialFunctionCall!(
   Steps=ScalarParameter() {defaultValue=512, parameterName=Steps}
   Plane alignment.=ScalarParameter() {defaultValue=1, parameterName=Plane Alignment}
  ) {materialFunction=/Volumetrics/Content/VolumeTextures/MaterialFunctions/VolumeBoxIntersect}
  n2 = Multiply(A=@n1.Box Distance, B=ScalarParameter() {defaultValue=512, parameterName=Steps})
  MaterialExpressionClamp_1 = Clamp(None=Floor(None=@n2)) {desc=MaxSteps, maxDefault=512}
  MaterialExpressionFrac_0 = Frac(None=@n2) {desc=FinalStepSize}
  MaterialExpressionMultiply_1 = Multiply(A=@n1.Intersection Position) {desc=Ray Entry}
  MaterialExpressionScalarParameter_9 = ScalarParameter() {defaultValue=16, parameterName=Density}
  MaterialExpressionMultiply_4 = Multiply(
   A=MaterialFunctionCall(Alpha Threshold=ScalarParameter() {defaultValue=0.5, parameterName=Jitter Alpha Threshold}) {materialFunction=&DitherTemporalAA}.Result
   B=ScalarParameter() {defaultValue=1, parameterName=Jitter}
  )
  MaterialExpressionScalarParameter_27 = ScalarParameter() {defaultValue=15000, parameterName=Temperature}
  MaterialExpressionVectorParameter_1 = VectorParameter() {parameterName=VolumeResolution, defaultValue=<rgba 256 256 256 0>}
  MaterialExpressionScalarParameter_32 = ScalarParameter() {parameterName=Use Blackbody or Grayscale}
  n3 = Custom!(
   Tex=@MaterialExpressionTextureObjectParameter_1
   XYFrames=@MaterialExpressionComponentMask_8
   StepSize=@MaterialExpressionDivide_1
   MaxSteps=@MaterialExpressionClamp_1
   FinalStep=@MaterialExpressionFrac_0
   CurPos=@MaterialExpressionMultiply_1
   Density=@MaterialExpressionScalarParameter_9
   Jitter=@MaterialExpressionMultiply_4
   Temp=@MaterialExpressionScalarParameter_27
   VolumeResolution=@MaterialExpressionVectorParameter_1.RGB
   UseBlackBodyColors=@MaterialExpressionScalarParameter_32
  ) {
   code="##define EnableCurl 0\r\nfloat numFrames = XYFrames.x * XYFrames.y;\r\nfloat accumdist = 0;\r\nfloat3 curdensity = 0;\r\nfloat transmittance = 1;\r\nfloat3 localcamvec = normalize( mul(Parameters.CameraVector, (float3x3)WSDemote(GetWorldToLocal(Parameters))) ) * StepSize;\r\n\r\nDensity *= StepSize;\r\nfloat3 lightenergy = 0;\r\n\r\nCurPos +=  localcamvec * Jitter;\r\n\r\nfor (int i = 0; i < MaxSteps; i++)\r\n{\t\r\n\tfloat3 samplepos =  saturate(CurPos);\r\n\tsamplepos.z = min(samplepos.z, 1 - (1/VolumeResolution.z));\r\n\r\n\tfloat4 cursample = PseudoVolumeTexture(Tex, TexSampler, samplepos, XYFrames, numFrames);\r\n\r\n\t//Sample Light Absorption and Scattering\r\n\tif( cursample.r + cursample.g + cursample.b > 0.00001)\r\n\t{\r\n\r\n\t\tcurdensity = dot(cursample.rgb, 0.333);\r\n\t\tcurdensity = 1 - exp(-curdensity * Density);\r\n\t\t\r\n\t\tfloat3 curenergy = UseBlackBodyColors == 1.0 ? MaterialExpressionBlackBody(Temp*cursample.a) : cursample.a;\r\n\t\t\r\n\t\tlightenergy += curenergy * curdensity * transmittance;\r\n\t\t\r\n\t\ttransmittance *= 1-curdensity;\r\n\t}\r\n\t\t\r\n\tCurPos -= localcamvec;\r\n}\r\n\r\nreturn float4( lightenergy, transmittance);\r\n"
   desc="\n*** Original source before expression upgrade ***\n##define EnableCurl 0\r\nfloat numFrames = XYFrames.x * XYFrames.y;\r\nfloat accumdist = 0;\r\nfloat3 curdensity = 0;\r\nfloat transmittance = 1;\r\nfloat3 localcamvec = normalize( mul(Parameters.CameraVector, (float3x3)GetPrimitiveData(Parameters).WorldToLocal) ) * StepSize;\r\n\r\nDensity *= StepSize;\r\nfloat3 lightenergy = 0;\r\n\r\nCurPos +=  localcamvec * Jitter;\r\n\r\nfor (int i = 0; i < MaxSteps; i++)\r\n{\t\r\n\tfloat3 samplepos =  saturate(CurPos);\r\n\tsamplepos.z = min(samplepos.z, 1 - (1/VolumeResolution.z));\r\n\r\n\tfloat4 cursample = PseudoVolumeTexture(Tex, TexSampler, samplepos, XYFrames, numFrames);\r\n\r\n\t//Sample Light Absorption and Scattering\r\n\tif( cursample.r + cursample.g + cursample.b > 0.00001)\r\n\t{\r\n\r\n\t\tcurdensity = dot(cursample.rgb, 0.333);\r\n\t\tcurdensity = 1 - exp(-curdensity * Density);\r\n\t\t\r\n\t\tfloat3 curenergy = UseBlackBodyColors == 1.0 ? MaterialExpressionBlackBody(Temp*cursample.a) : cursample.a;\r\n\t\t\r\n\t\tlightenergy += curenergy * curdensity * transmittance;\r\n\t\t\r\n\t\ttransmittance *= 1-curdensity;\r\n\t}\r\n\t\t\r\n\tCurPos -= localcamvec;\r\n}\r\n\r\nreturn float4( lightenergy, transmittance);\r\n"
   description=Density RayMarch
   outputType=CMOT_Float4
   inputs [
    {
     inputName=Tex
     input {expression=&self:MaterialExpressionTextureObjectParameter_1}
    }
    {
     inputName=XYFrames
     input {expression=&self:MaterialExpressionComponentMask_8}
    }
    {
     inputName=StepSize
     input {expression=&self:MaterialExpressionDivide_1}
    }
    {
     inputName=MaxSteps
     input {expression=&self:MaterialExpressionClamp_1}
    }
    {
     inputName=FinalStep
     input {expression=&self:MaterialExpressionFrac_0}
    }
    {
     inputName=CurPos
     input {expression=&self:MaterialExpressionMultiply_1}
    }
    {
     inputName=Density
     input {expression=&self:MaterialExpressionScalarParameter_9}
    }
    {
     inputName=Jitter
     input {expression=&self:MaterialExpressionMultiply_4}
    }
    {
     inputName=Temp
     input {expression=&self:MaterialExpressionScalarParameter_27}
    }
    {
     inputName=VolumeResolution
     input {
      expression=&self:MaterialExpressionVectorParameter_1
      mask=1
      maskB=1
      maskG=1
      maskR=1
     }
    }
    {
     inputName=UseBlackBodyColors
     input {expression=&self:MaterialExpressionScalarParameter_32}
    }
   ]
  }
  ComponentMask(None=@n3) {b=true, g=true, r=true}
  OneMinus(None=ComponentMask(None=@n3) {a=true})
 ]
}
