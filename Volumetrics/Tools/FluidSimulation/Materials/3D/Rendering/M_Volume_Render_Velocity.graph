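# Unlit alpha-composite volume material that visualizes fluid simulation velocity, with optional pressure, color-channel density mask and brush preview.
# A Custom HLSL node raymarches from the VolumeBoxIntersect entry through velocity, color and pressure flipbook pseudo-volume textures, with optional per-pixel jitter.
# It accumulates signed velocity as emissive energy alongside transmittance.
# Further nodes project the result onto view right/up, unpremultiply it by opacity and scale by volume resolution.
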Material /Volumetrics/Tools/FluidSimulation/Materials/3D/Rendering/M_Volume_Render_Velocity
graph {
 properties {
  bAllowNegativeEmissiveColor=true
  bDisableDepthTest=true
  blendMode=BLEND_AlphaComposite
  pixelDepthOffsetMode=PDOM_Legacy
  shadingModel=MSM_Unlit
  usedShadingModels=Unlit
 }
 nodes [
  n1 = MaterialFunctionCall!(Steps=ScalarParameter() {defaultValue=64, parameterName=Steps}, Plane alignment.=ScalarParameter() {parameterName=Plane Alignment}) {materialFunction=/Volumetrics/Content/VolumeTextures/MaterialFunctions/VolumeBoxIntersect}
  n2 = Multiply(A=@n1.Box Distance, B=ScalarParameter() {defaultValue=64, parameterName=Steps})
  Floor(None=@n2)
  MaterialExpressionTextureObjectParameter_3 = TextureObjectParameter() {parameterName=Velocity, texture=/Volumetrics/Content/VolumeTextures/Textures/RT_Default}
  MaterialExpressionComponentMask_2 = ComponentMask(None=VectorParameter() {parameterName=XYFrames, defaultValue=<rgba 16 16 0 0>}.RGB) {g=true, r=true}
  MaterialExpressionDivide_8 = Divide(B=ScalarParameter() {defaultValue=64, parameterName=Steps})
  MaterialExpressionClamp_0 = Clamp(None=Ceil(None=@n2)) {desc=MaxSteps, maxDefault=512}
  MaterialExpressionFrac_3 = Frac(None=@n2) {desc=FinalStepSize}
  MaterialExpressionMultiply_0 = Multiply(A=@n1.Intersection Position) {desc=Ray Entry}
  MaterialExpressionScalarParameter_0 = ScalarParameter() {defaultValue=1, parameterName=Density}
  MaterialExpressionScalarParameter_8 = ScalarParameter() {defaultValue=1, parameterName=VelocityAmount}
  MaterialExpressionScalarParameter_12 = ScalarParameter() {defaultValue=0.01, parameterName=SB}
  MaterialExpressionScalarParameter_13 = ScalarParameter() {parameterName=Jitter}
  MaterialExpressionScalarParameter_6 = ScalarParameter() {parameterName=PreviewBrush}
  MaterialExpressionTextureObjectParameter_2 = TextureObjectParameter() {parameterName=Color, texture=&RT_Default}
  MaterialExpressionScalarParameter_17 = ScalarParameter() {parameterName=Density Mask}
  MaterialExpressionVectorParameter_18 = VectorParameter() {parameterName=VolumeResolution, defaultValue=<rgba 256 256 256 0>}
  MaterialExpressionTextureObjectParameter_0 = TextureObjectParameter() {parameterName=Pressure, texture=&RT_Default}
  MaterialExpressionScalarParameter_7 = ScalarParameter() {parameterName=PressureAmount}
  n3 = Custom!(
   Tex=@MaterialExpressionTextureObjectParameter_3
   XYFrames=@MaterialExpressionComponentMask_2
   StepSize=@MaterialExpressionDivide_8
   MaxSteps=@MaterialExpressionClamp_0
   FinalStep=@MaterialExpressionFrac_3
   CurPos=@MaterialExpressionMultiply_0
   Density=@MaterialExpressionScalarParameter_0
   VelocityAmount=@MaterialExpressionScalarParameter_8
   ShadowThreshold=@MaterialExpressionScalarParameter_12
   Jitter=@MaterialExpressionScalarParameter_13
   PreviewAmount=@MaterialExpressionScalarParameter_6
   Tex2=@MaterialExpressionTextureObjectParameter_2
   DensityMask=@MaterialExpressionScalarParameter_17
   VolumeResolution=@MaterialExpressionVectorParameter_18.RGB
   Pressure=@MaterialExpressionTextureObjectParameter_0
   PressureAmount=@MaterialExpressionScalarParameter_7
  ) {
   code="float numFrames = XYFrames.x * XYFrames.y;\r\nfloat accumdist = 0;\r\nfloat curdensity = 0;\r\nfloat transmittance = 1;\r\nfloat3 localcamvec = normalize( mul(Parameters.CameraVector, (float3x3)WSDemote(GetWorldToLocal(Parameters))) ) * StepSize;\r\n\r\n\r\n\r\nDensity *= StepSize;\r\nfloat3 lightenergy = 0;\r\n\r\nint3 randpos = int3(Parameters.SvPosition.xy, View.StateFrameIndexMod8);\r\nfloat rand =float(Rand3DPCG16(randpos).x) / 0xffff;\r\nCurPos +=  localcamvec * rand.x * Jitter;\r\n\r\n\r\nfor (int i = 0; i < MaxSteps; i++)\r\n{\t\r\n\tfloat3 velocity = PseudoVolumeTexture(Tex, TexSampler, clamp( CurPos, 0.5 / VolumeResolution, 1 - (0.5 / VolumeResolution) ), XYFrames, numFrames).rgb;\r\n\tfloat2 color = PseudoVolumeTexture(Tex2, TexSampler,  clamp( CurPos, 0.5 / VolumeResolution, 1 - (0.5 / VolumeResolution) ), XYFrames, numFrames).bg;\r\n\tfloat pressure = PseudoVolumeTexture(Pressure, PressureSampler,  clamp( CurPos, 0.5 / VolumeResolution, 1 - (0.5 / VolumeResolution) ), XYFrames, numFrames);\r\n\t\r\n\tfloat alpha = lerp( saturate( length(velocity) ), saturate(color.r), DensityMask);\r\n\tfloat previewbrush = color.g * PreviewAmount;\r\n\r\n\t//float3 negativeswizzle = max(0, -velocity.zxy);\r\n\t//velocity = max(0, velocity);\r\n\t//velocity += negativeswizzle;\r\n\t\r\n\t//Sample Light Absorption and Scattering\r\n\tif( alpha + previewbrush > 0.001)\r\n\t{\r\n\r\n\t\tcurdensity = saturate( length(velocity) * Density) * alpha;\r\n\t\tlightenergy += curdensity * transmittance * (velocity) * VelocityAmount;\r\n\t\tlightenergy += previewbrush * transmittance * float3(0,0,0.1);\r\n\r\n\t\tlightenergy += saturate(pressure * PressureAmount);\r\n\t\t\r\n\t\ttransmittance *= 1-curdensity;\r\n\r\n\t}\r\n\t\t\r\n\r\n\tCurPos -= localcamvec;\r\n}\r\n\r\n\r\nreturn float4( lightenergy, transmittance);\r\n"
   desc="\n*** Original source before expression upgrade ***\nfloat numFrames = XYFrames.x * XYFrames.y;\r\nfloat accumdist = 0;\r\nfloat curdensity = 0;\r\nfloat transmittance = 1;\r\nfloat3 localcamvec = normalize( mul(Parameters.CameraVector, (float3x3)GetPrimitiveData(Parameters).WorldToLocal) ) * StepSize;\r\n\r\n\r\n\r\nDensity *= StepSize;\r\nfloat3 lightenergy = 0;\r\n\r\nint3 randpos = int3(Parameters.SvPosition.xy, View.StateFrameIndexMod8);\r\nfloat rand =float(Rand3DPCG16(randpos).x) / 0xffff;\r\nCurPos +=  localcamvec * rand.x * Jitter;\r\n\r\n\r\nfor (int i = 0; i < MaxSteps; i++)\r\n{\t\r\n\tfloat3 velocity = PseudoVolumeTexture(Tex, TexSampler, clamp( CurPos, 0.5 / VolumeResolution, 1 - (0.5 / VolumeResolution) ), XYFrames, numFrames).rgb;\r\n\tfloat2 color = PseudoVolumeTexture(Tex2, TexSampler,  clamp( CurPos, 0.5 / VolumeResolution, 1 - (0.5 / VolumeResolution) ), XYFrames, numFrames).bg;\r\n\tfloat pressure = PseudoVolumeTexture(Pressure, PressureSampler,  clamp( CurPos, 0.5 / VolumeResolution, 1 - (0.5 / VolumeResolution) ), XYFrames, numFrames);\r\n\t\r\n\tfloat alpha = lerp( saturate( length(velocity) ), saturate(color.r), DensityMask);\r\n\tfloat previewbrush = color.g * PreviewAmount;\r\n\r\n\t//float3 negativeswizzle = max(0, -velocity.zxy);\r\n\t//velocity = max(0, velocity);\r\n\t//velocity += negativeswizzle;\r\n\t\r\n\t//Sample Light Absorption and Scattering\r\n\tif( alpha + previewbrush > 0.001)\r\n\t{\r\n\r\n\t\tcurdensity = saturate( length(velocity) * Density) * alpha;\r\n\t\tlightenergy += curdensity * transmittance * (velocity) * VelocityAmount;\r\n\t\tlightenergy += previewbrush * transmittance * float3(0,0,0.1);\r\n\r\n\t\tlightenergy += saturate(pressure * PressureAmount);\r\n\t\t\r\n\t\ttransmittance *= 1-curdensity;\r\n\r\n\t}\r\n\t\t\r\n\r\n\tCurPos -= localcamvec;\r\n}\r\n\r\n\r\nreturn float4( lightenergy, transmittance);\r\n"
   description=Density RayMarch
   outputType=CMOT_Float4
   inputs [
    {
     inputName=Tex
     input {expression=&self:MaterialExpressionTextureObjectParameter_3}
    }
    {
     inputName=XYFrames
     input {expression=&self:MaterialExpressionComponentMask_2}
    }
    {
     inputName=StepSize
     input {expression=&self:MaterialExpressionDivide_8}
    }
    {
     inputName=MaxSteps
     input {expression=&self:MaterialExpressionClamp_0}
    }
    {
     inputName=FinalStep
     input {expression=&self:MaterialExpressionFrac_3}
    }
    {
     inputName=CurPos
     input {expression=&self:MaterialExpressionMultiply_0}
    }
    {
     inputName=Density
     input {expression=&self:MaterialExpressionScalarParameter_0}
    }
    {
     inputName=VelocityAmount
     input {expression=&self:MaterialExpressionScalarParameter_8}
    }
    {
     inputName=ShadowThreshold
     input {expression=&self:MaterialExpressionScalarParameter_12}
    }
    {
     inputName=Jitter
     input {expression=&self:MaterialExpressionScalarParameter_13}
    }
    {
     inputName=PreviewAmount
     input {expression=&self:MaterialExpressionScalarParameter_6}
    }
    {
     inputName=Tex2
     input {expression=&self:MaterialExpressionTextureObjectParameter_2}
    }
    {
     inputName=DensityMask
     input {expression=&self:MaterialExpressionScalarParameter_17}
    }
    {
     inputName=VolumeResolution
     input {
      expression=&self:MaterialExpressionVectorParameter_18
      mask=1
      maskB=1
      maskG=1
      maskR=1
     }
    }
    {
     inputName=Pressure
     input {expression=&self:MaterialExpressionTextureObjectParameter_0}
    }
    {
     inputName=PressureAmount
     input {expression=&self:MaterialExpressionScalarParameter_7}
    }
   ]
  }
  n4 = OneMinus(None=ComponentMask(None=@n3) {a=true})
  n5 = Constant() {r=0}
  n6 = Constant3Vector() {constant=<rgba 1 0 0 0>}
  n7 = ComponentMask(None=@n3) {b=true, g=true, r=true}
  n8 = Constant3Vector() {constant=<rgba 0 1 0 0>}
  n9 = MaterialFunctionCall(
   X=DotProduct(A=Transform(None=@n6) {desc=View Right, transformSourceType=TRANSFORMSOURCE_View}, B=@n7)
   Y=DotProduct(A=Transform(None=@n8) {desc=View Up, transformSourceType=TRANSFORMSOURCE_View}, B=@n7)
  ) {materialFunction=&MakeFloat3}
  n10 = Divide(
   A=Divide(A=@n9.Result, B=Max(A=@n4, B=@n5)) {desc=Velocity In Texels}
   B=VectorParameter() {parameterName=VolumeResolution, defaultValue=<rgba 256 256 256 0>}.RGB
  ) {desc=velocity in Frames}
  LinearInterpolate(A=@n9.Result, B=@n10, Alpha=ScalarParameter() {defaultValue=1, parameterName=Unpremultiply})
  n11 = ScalarParameter() {defaultValue=0.1, desc=Max travel per frame - Must Match M_Draw_SubUV_Velocity Material Value, parameterName=Velocity Encoding Max}
  Multiply(A=@n10, B=Divide(B=@n11))
  ScalarParameter() {defaultValue=16, parameterName=XYFrames}
 ]
}
